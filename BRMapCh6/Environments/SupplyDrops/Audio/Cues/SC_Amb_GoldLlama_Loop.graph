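# Looping ambient sound cue for the gold llama supply drop, using the loot ambient loop sound class.
# Two looping gold llama wave layers play together through a mixer, which boosts both inputs equally.
# The mix feeds a root modulator left at its default settings.
# A shared mono item ambient-loop attenuation asset handles spatial falloff.
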SoundCue /BRMapCh6/Environments/SupplyDrops/Audio/Cues/SC_Amb_GoldLlama_Loop
root=Modulator_0
nodes [
 {id=Mixer_0, children [WavePlayer_0,WavePlayer_1], properties {inputVolume [2,2]}}
 {id=Modulator_0, children [Mixer_0]}
 {id=WavePlayer_0, properties {bLooping=true, soundWaveAssetPtr=/BRMapCh6/Environments/SupplyDrops/Audio/Waves/GoldLlama_Ambient_Loop_01}}
 {id=WavePlayer_1, properties {bLooping=true, soundWaveAssetPtr=/BRMapCh6/Environments/SupplyDrops/Audio/Waves/GoldLlama_Ambient_Loop_02}}
]
properties {
 attenuationSettings=/Game/Sounds/Attenuation/Fort_Item_Ambient_Loop_Mono_Default_Att
 maxDistance=1250
 soundClassObject=/Game/Sounds/FortSoundClassesAndModes/Fort_Loot_Ambient_Loop
 volumeMultiplier=1
}
