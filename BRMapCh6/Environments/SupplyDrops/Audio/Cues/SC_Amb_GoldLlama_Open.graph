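# Sound cue that plays the opening sound for the Gold Llama supply drop.
# A Random node picks one of three equally weighted open waves, and a Modulator above it varies each play.
# It uses the item-interact attenuation and the loot reward sound class, with boosted volume.
# The Random node is excluded from branch culling, so all variations stay available.
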SoundCue /BRMapCh6/Environments/SupplyDrops/Audio/Cues/SC_Amb_GoldLlama_Open
root=Modulator_0
nodes [
 {id=Modulator_0, children [Random_0]}
 {id=Random_0, children [WavePlayer_2,WavePlayer_3,WavePlayer_4], properties {bShouldExcludeFromBranchCulling=true, weights [1,1,1]}}
 {id=WavePlayer_2, properties {soundWaveAssetPtr=/BRMapCh6/Environments/SupplyDrops/Audio/Waves/SupplyDrop_GoldLlama_Open_01}}
 {id=WavePlayer_3, properties {soundWaveAssetPtr=/BRMapCh6/Environments/SupplyDrops/Audio/Waves/SupplyDrop_GoldLlama_Open_02}}
 {id=WavePlayer_4, properties {soundWaveAssetPtr=/BRMapCh6/Environments/SupplyDrops/Audio/Waves/SupplyDrop_GoldLlama_Open_03}}
]
properties {
 attenuationSettings=/Game/Sounds/Attenuation/Fort_Item_Interact_Att
 maxDistance=1850
 soundClassObject=&Fort_Loot_Reward_PSM
 volumeMultiplier=2
}
